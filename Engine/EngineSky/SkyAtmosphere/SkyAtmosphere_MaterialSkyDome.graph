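# Unlit sky dome material that draws the Sky Atmosphere sky and sun disk with a lit 2D cloud layer on top.
# Long-lat cloud mask samples, offset sunward (dithered) and vertically, fake self-shadowing through exponential falloff.
# Atmosphere sunlight, shaped by Schlick phase, edge glow and an earth-shadow fade, adds to shadowed ambient scattered light.
# Exponential cloud opacity, masked at the horizon by mesh UV, blends clouds over sky before a final tint.
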Material /Engine/EngineSky/SkyAtmosphere/SkyAtmosphere_MaterialSkyDome
graph {
 properties {bIsSky=true, pixelDepthOffsetMode=PDOM_Legacy, shadingModel=MSM_Unlit, usedShadingModels=Unlit}
 nodes [
  n1 = SkyAtmosphereLightDiskLuminance!(DiskAngularDiameterOverride)
  n2 = SkyAtmosphereViewLuminance!(WorldDirection)
  n3 = SkyAtmosphereLightDirection()
  n4 = Divide(
   A=Add(A=ComponentMask(None=@n3) {b=true}, B=ScalarParameter() {defaultValue=0.075, parameterName=EarthShadowStart})
   B=ScalarParameter() {defaultValue=0.05, parameterName=EarthShadowLength}
  )
  n5 = WorldPosition()
  n6 = SkyAtmosphereLightIlluminance!(World Position=AppendVector(A=ComponentMask(None=@n5.XYZ) {g=true, r=true}, B=ScalarParameter() {defaultValue=5000, parameterName=Cloud Altitude for Sun light}))
  n7 = MaterialFunctionCall!(
   Cosine=DotProduct(A=SkyAtmosphereLightDirection(), B=CameraVectorWS())
   Anisotropy=ScalarParameter() {defaultValue=0.8, parameterName=Anisotropy}
  ) {materialFunction=/Engine/Functions/Engine_MaterialFunctions01/Volumetrics/SchlickPhase}
  n8 = ScalarParameter() {defaultValue=3, parameterName=Edge Glow Amount}
  n9 = Normalize(VectorInput=WorldPosition().XYZ)
  n10 = Add(
   A=Multiply(
    A=MaterialFunctionCall!(Vector=@n9) {materialFunction=/Engine/Functions/Engine_MaterialFunctions01/Coordinates/LongLatToUV}.Result
    B=Constant2Vector() {g=2, r=1}
   )
   B=Constant2Vector() {g=-1}
  )
  n11 = TextureSample!(UVs=@n10, Tex, Level=ScalarParameter() {defaultValue=0.5, parameterName=Cloud Shading Mip}, Apply View MipBias) {desc=Center Sample, mipValueMode=TMVM_MipLevel, samplerType=SAMPLERTYPE_Grayscale, texture=/Engine/MapTemplates/T_UDK_Sky_CloudMask03}
  MaterialExpressionMultiply_25 = Multiply(A=ScalarParameter() {defaultValue=8, parameterName=Edge Lit Density}, B=@n11.R)
  n12 = Custom!(x=@MaterialExpressionMultiply_25) {
   code="return exp(-x);"
   description=Exp
   inputs [{inputName=x, input {expression=&self:MaterialExpressionMultiply_25}}]
  }
  n13 = Add(
   A=Multiply(A=ScalarParameter() {defaultValue=3, parameterName=Sun Intensity Boost}, B=@n7.Result)
   B=Multiply(A=@n7.Result, B=Multiply(A=@n8, B=@n12))
  )
  n14 = Multiply(A=SkyAtmosphereLightDirection(), B=VectorParameter() {parameterName=Light Vector Squash, defaultValue=<rgba 1 1 0.4 0>}.RGB)
  n15 = Subtract(A=@n9, B=@n14)
  n16 = MaterialFunctionCall() {materialFunction=&DitherTemporalAA}
  n17 = Add(
   A=ScalarParameter() {defaultValue=-0.06, parameterName=Light Trace Distance}
   B=Multiply(A=Subtract(A=@n16.Result) {constB=0.5}, B=ScalarParameter() {defaultValue=0.01, parameterName=Light Distance Dither})
  )
  n18 = MaterialFunctionCall(Vector=Add(A=@n9, B=Multiply(A=@n15, B=@n17) {constB=-1})) {materialFunction=&LongLatToUV}
  n19 = Constant2Vector() {g=2, r=1}
  n20 = TextureSample(
   UVs=Add(A=Multiply(A=@n18.Result, B=@n19), B=Constant2Vector() {g=-1})
   Level=ScalarParameter() {defaultValue=0.5, parameterName=Cloud Shading Mip}
  ) {desc=Sun Offset, mipValueMode=TMVM_MipLevel, samplerType=SAMPLERTYPE_Grayscale, texture=&T_UDK_Sky_CloudMask03}
  n21 = Add(A=Multiply(A=@n7.Result, B=ScalarParameter() {defaultValue=-1, parameterName=AnisoPush}))
  n22 = Add(A=Subtract(A=Divide(A=@n20.R, B=@n21), B=@n11.R), B=ScalarParameter() {defaultValue=1.5, parameterName=ShadowBias})
  n23 = VertexNormalWS()
  n24 = Add(B=Multiply(A=DotProduct(A=@n14, B=@n23), B=ScalarParameter() {defaultValue=-0.75, parameterName=VertexFacingOffset}))
  MaterialExpressionMax_2 = Max(A=Multiply(A=Add(A=@n22, B=@n24), B=ScalarParameter() {defaultValue=1, parameterName=ShadowDensity})) {constB=0}
  n25 = Custom(x=@MaterialExpressionMax_2) {
   code="return exp(-x);"
   description=Exp
   inputs [{inputName=x, input {expression=&self:MaterialExpressionMax_2}}]
  }
  n26 = Multiply(A=Multiply(A=Saturate(None=@n4), B=Multiply(A=@n6, B=@n13)), B=@n25)
  n27 = Subtract(A=@n9, B=Constant3Vector() {constant=<rgba 0 0 1 0>})
  n28 = ScalarParameter() {defaultValue=-0.05, parameterName=Ambient Offset}
  n29 = MaterialFunctionCall(Vector=Add(A=@n9, B=Multiply(A=@n27, B=@n28) {constB=-1})) {materialFunction=&LongLatToUV}
  n30 = Constant2Vector() {g=2, r=1}
  n31 = TextureSample(
   UVs=Add(A=Multiply(A=@n29.Result, B=@n30), B=Constant2Vector() {g=-1})
   Level=ScalarParameter() {defaultValue=0.5, parameterName=Cloud Shading Mip}
  ) {desc=Ambient Offset, mipValueMode=TMVM_MipLevel, samplerType=SAMPLERTYPE_Grayscale, texture=&T_UDK_Sky_CloudMask03}
  n32 = Subtract(A=@n31.R, B=@n11.R)
  n33 = ScalarParameter() {defaultValue=1, parameterName=AmbientShadowBias}
  MaterialExpressionMax_3 = Max(A=Multiply(A=Add(A=@n32, B=@n33), B=ScalarParameter() {defaultValue=0.5, parameterName=AmbientShadowDensity})) {constB=0}
  n34 = Custom(x=@MaterialExpressionMax_3) {
   code="return exp(-x);"
   description=Exp
   inputs [{inputName=x, input {expression=&self:MaterialExpressionMax_3}}]
  }
  n35 = Multiply(
   A=Multiply(A=SkyAtmosphereDistantLightScatteredLuminance(), B=ScalarParameter() {defaultValue=1, parameterName=Cloud Ambient Lighting})
   B=@n34
  )
  n36 = TextureSample(UVs=@n10, Level=ScalarParameter() {parameterName=Cloud Opacity Mip}) {mipValueMode=TMVM_MipLevel, samplerType=SAMPLERTYPE_Grayscale, texture=&T_UDK_Sky_CloudMask03}
  n37 = ScalarParameter() {defaultValue=0.03, parameterName=Alpha Offset}
  MaterialExpressionMax_4 = Max(A=Multiply(
    A=MaterialFunctionCall!(Alpha=@n36.R, Offset=@n37) {materialFunction=/Engine/Functions/Engine_MaterialFunctions01/Opacity/AlphaOffset}.Result
    B=ScalarParameter() {defaultValue=1, parameterName=Opacity Density}
   )) {constB=0}
  n38 = Custom(x=@MaterialExpressionMax_4) {
   code="return 1-exp(-x);"
   description=Exp
   inputs [{inputName=x, input {expression=&self:MaterialExpressionMax_4}}]
  }
  n39 = Subtract(A=ComponentMask(None=TextureCoordinate()) {g=true}, B=OneMinus(None=ScalarParameter() {defaultValue=0.1, parameterName=Horizon Mask UV Min}))
  n40 = ScalarParameter() {defaultValue=0.1, parameterName=Horizon Mask Width}
  n41 = OneMinus(None=Saturate(None=Divide(A=@n39, B=@n40)))
  Multiply(
   A=LinearInterpolate(A=Add(A=@n1, B=@n2), B=Add(A=@n26, B=@n35), Alpha=Multiply(A=@n38, B=@n41))
   B=VectorParameter() {parameterName=asdf}.RGB
  )
 ]
}
